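; Broken Dragon Gate environment actor that shows the portal's gold-tendril "after" VFX and repels what touches it.
; Construction, client match/playlist readiness and the server-set bSetToAfterState RepNotify drive Portal_Phase in a material parameter collection.
; On authority, box overlaps apply a knockback GameplayEffect to non-NPC player pawns.
; Otherwise pickups get tossed by a scalable force along a vector built from look-at rotator angles, not a normalized direction.
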(event EventBeginPlay
  (Class|FortAsyncActionAuthorityMatchReady|MatchReadyAsync
    (:OnReady
      (Class|FortAsyncActionCurrentPlaylistReady|PlaylistDataReadyAsync
        (:OnReady
          (if (Networking|IsServer)
            (Variables|Default|SetbSetToAfterState true)
            (else
              (CallFunction|SetPortalPhaseToAfterState))))))))

(event Custom|Set%20Portal%20Phase%20to%20After%20State
  (Rendering|Material|SetScalarParameterValue (Variables|Default|GetMPC_GrillMoon) (Utilities|Name|MakeLiteralName "Portal_Phase") 6.0 :!class /Script/Engine.KismetMaterialLibrary) ; Phase 6 - the "after" state of the portal vfx\nthat shows gold tendril energy
)

(event OnComponentBeginOverlap(Box-ImpulseCollision)
  (params (OverlappedComponent :type Primitive%20Component%20Object%20Reference) (OtherActor :type Actor%20Object%20Reference) (OtherComp :type Primitive%20Component%20Object%20Reference) (OtherBodyIndex :type Integer) (bFromSweep :type Boolean) (SweepResult :type Hit%20Result%20Structure%20%28by%20ref%29))
  (Networking|SwitchHasAuthority
    (:Authority
      (bind _asfort_player_pawn_athena (Utilities|Casting|CastToFortPlayerPawnAthena OtherActor)
        (:then
          (if (not (GameplayTags|HasMatchingGameplayTag _asfort_player_pawn_athena "(TagName=\"Pawn.Athena.NPC\")"))
            (GameplayEffects|FortApplyGameplayEffectToSelf :self (Ability|GetFortAbilitySystemComponent _asfort_player_pawn_athena) :GameplayEffectClass "/GrillMoon/GameplayCues/GE_BrokenPortal_GrantKnockback.GE_BrokenPortal_GrantKnockback_C" :OptionalSourceObject self :OptionalEffectCauser self :Level 0.0)))
        (:CastFailed
          (bind _asfort_pickup (Utilities|Casting|CastToFortPickup OtherActor))
          (bind (_returnvalue_roll _returnvalue_pitch _returnvalue_yaw) (Math|Rotator|BreakRotator (Math|Rotator|FindLookatRotation (Transformation|GetActorLocation) (Transformation|GetActorLocation _asfort_pickup))))
          (Pickup|TossPickupinDirection _asfort_pickup (* (Math|Vector|MakeVector _returnvalue_roll _returnvalue_pitch (Math|Float|Absolute(Float) _returnvalue_yaw)) (ScalableFloat|GetValueatLevel (Variables|Default|GetPickupTossForce_Hf))) 0 false true))))))

(fn OnRep_bSetToAfterState
  (CallFunction|SetPortalPhaseToAfterState))

(fn ConstructionScript
  (CallFunction|SetPortalPhaseToAfterState))
